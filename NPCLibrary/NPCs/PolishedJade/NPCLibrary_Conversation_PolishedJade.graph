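# Conversation database for the Polished Jade NPC, defining the greeting and top-level branches players get when talking to them.
# The entry tag leads to one speech node whose default line can be replaced by prioritized contextual lines for first-ever meetings, per-match conversation count, or specific player outfits addressed as "brother".
# That node fans out through link nodes to separate NPC quest, basic services, Polished Jade dialogue and Season 37 story quest entries.
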ConversationDatabase /NPCLibrary/NPCs/PolishedJade/NPCLibrary_Conversation_PolishedJade
entry_points [{tag=Conversation.Entry.PolishedJade, destinations [#1]}]
nodes [
 ConversationEntryPointNode_7 {
  id=#1
  class=ConversationEntryPointNode
  outputs [#2]
  properties {entryTag {tagName=Conversation.Entry.PolishedJade}}
 }
 Speech_C_4 {
  id=#2
  class=Speech_C
  outputs [#3, #4, #5, #6]
  properties {
   nodeName=""
   generalConfig {
    defaultMessage=These Bugs are NOTHING like the Sprites...
    contextualMessages [
     {
      message="So... Bugs, huh?"
      contextRequirements [{requirement {class=FortControllerRequirement_IsFirstEverConversationWithNPC}}]
      priority {curve {rowName=Tandem.Speech.Priority.FirstEverInteraction}}
     }
     {
      message="Hi!"
      contextRequirements [{requirement {class=FortControllerRequirement_CheckNPCMatchConversationCount, properties {nPCUniqueIdTag {tagName=AISpawnerData.Type.Tandem.PolishedJade}}}}]
      priority {curve {rowName=None}}
     }
     {
      message=Good riddance to those Bugs.
      contextRequirements [{requirement {class=FortControllerRequirement_CheckNPCMatchConversationCount, properties {nPCUniqueIdTag {tagName=AISpawnerData.Type.Tandem.PolishedJade}}}}]
      priority {curve {rowName=None}}
     }
     {
      message="You might be the only one of us to resist the Spirit Realm's influence, brother."
      contextRequirements [{requirement {class=FortControllerRequirement_HasCID, properties {allowedCIds [&Character_CubicVice_Dual]}}}]
      priority {value=2, curve {curveTable=null, rowName=None}}
     }
     {
      message="It's good to have you back, brother."
      contextRequirements [{requirement {
         class=FortControllerRequirement_HasCID
         properties {allowedCIds [&Character_MistMylar_Claw, /BRCosmetics/Athena/Items/Cosmetics/Characters/Character_SolidMist]}
        }}]
      priority {value=2, curve {curveTable=null, rowName=None}}
     }
    ]
   }
  }
 }
 ConversationLinkNode_3 {
  id=#6
  class=ConversationLinkNode
  properties {remoteEntryTag {tagName=Conversation.Entry.Season37.Story.P03.Q01.B}}
 }
 ConversationLinkNode_2 {
  id=#5
  class=ConversationLinkNode
  properties {remoteEntryTag {tagName=Conversation.Entry.PolishedJade.Dialogue}}
 }
 ConversationLinkNode_0 {id=#4, class=ConversationLinkNode, properties {remoteEntryTag {tagName=Conversation.Entry.BasicServices}}}
 ConversationLinkNode_1 {id=#3, class=ConversationLinkNode, properties {remoteEntryTag {tagName=Conversation.Entry.NPCLibrary.Quests}}}
]
